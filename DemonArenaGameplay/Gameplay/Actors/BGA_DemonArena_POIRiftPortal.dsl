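; Demon Arena rift portal actor that snaps itself to the ground when spawned, leaving most rift behavior to its parent Blueprint.
; BeginPlay sets the permanent-rift flag from a scalable float before calling the parent's logic.
; Once playlist data is ready, the server traces downward over a tunable distance and moves the rift to the hit, relying on replicated location.
; If no ground is found, a short lifespan destroys the rift.
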(event EventBeginPlay
  ;[0.05316 1 0.08408] BeginPlay
  ; Snap the rift to the ground - Parent BP handles most necessary logic
  (Variables|Default|SetPermanentRift (ScalableFloat|GetValueasBool (Variables|RiftPortal|GetHF_DemonRiftsArePermanentBool))) ; set permanent variable before \nthe parent logic runs
  (|CallParentFunction)
  (Class|FortAsyncActionCurrentPlaylistReady|PlaylistDataReadyAsync
    (:OnReady
      (Networking|SwitchHasAuthority
        (:Authority
          ;;[0.4052 0.7052 1] Check for the nearest ground/world collision to move the Rift to
          (bind (_outhit _returnvalue) (Collision|LineTraceByChannel (Transformation|GetActorLocation) (- (Transformation|GetActorLocation) (Math|Vector|MakeVector 0.0 0.0 (ScalableFloat|GetValueatLevel (Variables|RiftPortal|GetHF_DistanceToCheckForGround)))) "TraceTypeQuery10" true))
          (if _returnvalue
            ;;[0 0.08485 1] Move the rift to the Ground
            (Transformation|SetActorLocation (.Location (Collision|BreakHitResult _outhit)) false true) ; location is replicated to clients
            (else
              ;;[1 0.2327 0] Destroy the Rift
              (Actor|SetLifeSpan 0.25) ; these rifts could not spawn\non the ground. Destroy this rift
            )))))))

(fn ConstructionScript
  (|CallParentFunction))
