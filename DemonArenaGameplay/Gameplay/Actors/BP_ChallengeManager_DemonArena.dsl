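; Challenge manager Blueprint for the Demon Arena, tracking boss progress during the final fight.
; HandleBossHealthChanged checks that the boss is ready for the final fight and its actor is valid.
; It then casts the actor to the non-participant boss pawn and sets boss progress from its health percent.
; The event graph holds only empty stub events, and the construction script just calls its parent.
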(event EventBeginPlay
  (|CallParentFunction) ; This node is disabled and will not be called.\nDrag off pins to build functionality.
)

(event Collision|EventActorBeginOverlap
  (params (OtherActor :type Actor%20Object%20Reference)))

(event EventTick
  (params (DeltaSeconds :type Float%20%28single-precision%29)))

(fn HandleBossHealthChanged
  (bind _bossactor (Variables|Boss|GetBossActor))
  (if (Variables|Default|GetBossIsReadyForFinalFight)
    (Utilities|IsValid _bossactor
      (:Is%20Valid
        (bind _asbp_player_pawn_non_participant_boss (Utilities|Casting|CastToBP_PlayerPawn_NonParticipant_Boss _bossactor))
        (Variables|Progress|SetBossProgress (FortPawn|HealthPercent _asbp_player_pawn_non_participant_boss))))))

(fn ConstructionScript
  (|CallParentFunction))
