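; Timed UI helper actor for Demon Arena that marks a rift location through objective auto-marking, a minimap icon and a sound indicator.
; On BeginPlay the server sets its lifetime from a rift ping duration, registers the objective and starts the looping burst sound; EndPlay undoes registration and sound.
; Show Dynamic Land swaps in the land icon and legend, registers yaw-offset dynamic land on the game state and stops the sound.
; Hide From Mini Map hides the indicator on authority and forces a net update.
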(event EventBeginPlay
  (Networking|SwitchHasAuthority
    (:Authority
      (CallFunction|SetCustomTimetoLive (ScalableFloat|GetValueatLevel (Variables|Default|GetHF_RiftPingDuration)))
      (Class|BMMObjObjectiveActorComponent|RegisterObjectiveforAutoMarking (Variables|Default|GetB_MMObj_ObjectiveActorComponent))
      (Networking|FlushNetDormancy)
      (SoundIndicator|StartSoundIndicatorLooping (Variables|Default|GetFortSoundIndicator) (Transformation|GetActorLocation) (Variables|Default|GetRiftBurstSound) 1.0 0))))

(event EventEndPlay
  (params (EndPlayReason :type EEndPlayReason%20Enum :default "Destroyed"))
  (Networking|SwitchHasAuthority
    (:Authority
      (Class|BMMObjObjectiveActorComponent|UnregisterObjectiveforAutoMarking (Variables|Default|GetB_MMObj_ObjectiveActorComponent))
      (Networking|FlushNetDormancy)
      (SoundIndicator|StopSoundIndicatorLooping (Variables|Default|GetFortSoundIndicator)))))

(event Custom|Show%20Dynamic%20Land
  (bind (_returnvalue_roll _returnvalue_pitch _returnvalue_yaw) (Math|Rotator|BreakRotator (Transformation|GetActorRotation self)))
  (MiniMapData|SetMiniMapIcon (Variables|Default|GetFortMiniMap) (Widget|Brush|GetBrushResource (Variables|Default|GetLand%20Icon)))
  (MiniMapData|SetMiniMapIconMapLegendText (Variables|Default|GetFortMiniMap) (Variables|Default|GetLand%20Text))
  (Class|FortGameStateAthena|SetDynamicLand (Fortnite|Athena|GameState|GetGameStateAthena) (Utilities|String|StringToName (Utilities|GetObjectName self)) (Transformation|GetActorLocation self) (+ _returnvalue_yaw (Variables|Default|GetLandIconRotationOffset)) (Variables|Default|GetLand%20Icon) true)
  (SoundIndicator|StopSoundIndicatorLooping (Variables|Default|GetFortSoundIndicator)))

(event Custom|Hide%20From%20Mini%20Map
  (Networking|SwitchHasAuthority
    (:Authority
      (MiniMapData|SetMiniMapIndicatorIsVisible (Variables|Default|GetFortMiniMap) false)
      (Networking|ForceNetUpdate))))

(fn ConstructionScript)
