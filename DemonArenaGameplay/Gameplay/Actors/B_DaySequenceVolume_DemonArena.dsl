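; B_DaySequenceVolume_DemonArena is a Demon Arena day sequence volume Blueprint that strips the box shape inherited from its parent.
; Its construction script runs the parent's script, then destroys the Box component.
; A TODO there calls for nativizing this into a component to avoid being stuck with the parent class's shape.
; BeginPlay, overlap and Tick events are empty or disabled stubs, so the graphs add no runtime logic.
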(event EventBeginPlay
  (|CallParentFunction) ; This node is disabled and will not be called.\nDrag off pins to build functionality.
)

(event Collision|EventActorBeginOverlap
  (params (OtherActor :type Actor%20Object%20Reference)))

(event EventTick
  (params (DeltaSeconds :type Float%20%28single-precision%29)))

(fn ConstructionScript
  ; TODO PamF: We need to nativize this into a component so we're not stuck with the parent class' shape
  (|CallParentFunction)
  (Components|DestroyComponent (Variables|Default|GetBox)))
